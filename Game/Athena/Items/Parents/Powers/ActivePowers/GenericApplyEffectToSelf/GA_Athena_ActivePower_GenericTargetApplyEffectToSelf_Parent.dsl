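; Parent gameplay ability for active powers that apply a configurable gameplay effect to the owner, driven by cast montage events.
; On the apply-effect notify tag it runs the commit check, then plays a burst gameplay cue and applies the effect spec, storing the handle.
; Spec creation is an overridable function.
; The ability ends on montage blend-out, interruption, end notify, commit failure, or optionally after applying.
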(event EventCastMontageBlendOut
  ;[1 1 1] MontageBlendOut - End Ability
  (|CallParentFunction)
  (Ability|EndAbility))

(event EventCastMontageNotifyEndAbility
  ;[1 1 1] MontageEndNotify - End Ability
  (|CallParentFunction)
  (Ability|EndAbility))

(event EventCastMontageCancelledOrInterrupted
  ;[1 1 1] Montage Cancelled or Interrupted - end ability
  (|CallParentFunction)
  (Ability|EndAbility))

(event EventPowerCommitFailed
  ;[1 1 1] Commit Failed - End Ability
  (|CallParentFunction)
  (Ability|EndAbility))

(event Custom|ApplyEffectToSelf
  ;[1 1 1] Apply Effect To Self - Play burst GC and Apply Effect
  (CallFunction|PlayBurstGC (Variables|PowerGameplayCues|GetGC_ApplyEffect) (Variables|PowerGameplayCues|GetApplyEffectGCUsesParams) (CallFunction|ApplyEffectGCParams))
  (bind _returnvalue (CallFunction|MakeEffectSpec))
  (bind _returnvalue_2 (Ability|ApplyGameplayEffectSpecToOwner _returnvalue))
  (Variables|PowerEffectApply|SetGEH_ToApply _returnvalue_2)
  (if (Variables|PowerEffectApply|GetEndAbilityAfterApply)
    (Ability|EndAbility)))

(event EventCastMontageNotifyStart
  (params (NotifyTag :type Gameplay%20Tag%20Structure))
  ;[1 1 1] Event Montage Notify Start - Do the commit then Apply effect to self if passes the commit.
  (|CallParentFunction NotifyTag)
  (if (GameplayTags|Equal(GameplayTag) NotifyTag (Variables|PowerEffectApply|GetTag_ApplyEffectNotify))
    (Utilities|PowerCommitAbility (Variables|AbilityCommit|GetIgnoreAbilityCommit)
      (:Pass
        _)
      (:Fail
        (Development|PrintString "Power: Ability Commit Failed")
        (CallFunction|PowerCommitFailed))
      (:Ignored
        _))
    (CallFunction|ApplyEffecttoSelf)
    (elif (and (GameplayTags|Equal(GameplayTag) NotifyTag (Variables|PowerAnimationCast|GetT_AbilityEndCastNotify)) (GameplayTags|NotEqual(GameplayTag) :A (Variables|PowerAnimationCast|GetT_AbilityEndCastNotify)))
      (CallFunction|CastMontageNotifyEndAbility))))

(fn ApplyEffectGCParams
  (return 0))

(fn MakeEffectSpec
  (bind _ge_toapply (Variables|PowerEffectApply|GetGE_ToApply))
  ;[1 1 1] Make the effect spec this can be overridden if more work to the spec is needed.
  (if (Utilities|IsValidClass _ge_toapply)
    (return (Ability|MakeOutgoingGameplayEffectSpec _ge_toapply (Math|Conversions|ToFloat(Integer) (Variables|PowerEffectApply|GetEffectLevel))))
    (else
      (Development|PrintString "Power - The Effect that is trying to be applied is not a valid class."))))
